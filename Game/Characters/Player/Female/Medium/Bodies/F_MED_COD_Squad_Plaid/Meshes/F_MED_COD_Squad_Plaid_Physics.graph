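# Physics asset for the F_MED_COD_Squad_Plaid female player body, simulating a dangling tie piece at the pelvis.
# A small sphere on the tie base links to the pelvis capsule with free swing and twist.
# Top and bottom chains of two capsules each branch from that base, with tight swing and twist limit angles on every link.
# Only the pelvis and tie bones have bodies.
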PhysicsAsset /Game/Characters/Player/Female/Medium/Bodies/F_MED_COD_Squad_Plaid/Meshes/F_MED_COD_Squad_Plaid_Physics
bodies {
 dyn_tie_base [{shapeType=Sphere, radius=0.505}]
 dyn_tie_bot_01 [{shapeType=Capsule, length=4, radius=1, center=<2 0 0>, rotation=<rot 90 0 0>}]
 dyn_tie_bot_02 [{shapeType=Capsule, length=4, radius=1, center=<3 0 0>, rotation=<rot 90 0 0>}]
 dyn_tie_top_01 [{shapeType=Capsule, length=4, radius=1, center=<2 0 0>, rotation=<rot 90 0 0>}]
 dyn_tie_top_02 [{shapeType=Capsule, length=4, radius=1, center=<2 0 0>, rotation=<rot 90 0 0>}]
 pelvis [{shapeType=Capsule, length=9.25, radius=12, center=<0 -1.2113 0>}]
}
constraints [
 {bone1Name=dyn_tie_bot_01, bone2Name=dyn_tie_base, swing1LimitDegrees=15, swing2LimitDegrees=15, twistLimitDegrees=10}
 {bone1Name=dyn_tie_bot_02, bone2Name=dyn_tie_bot_01, swing1LimitDegrees=15, swing2LimitDegrees=15, twistLimitDegrees=10}
 {bone1Name=dyn_tie_top_01, bone2Name=dyn_tie_base, swing1LimitDegrees=15, swing2LimitDegrees=15, twistLimitDegrees=10}
 {bone1Name=dyn_tie_top_02, bone2Name=dyn_tie_top_01, swing1LimitDegrees=15, swing2LimitDegrees=15, twistLimitDegrees=10}
 {
  bone1Name=dyn_tie_base
  bone2Name=pelvis
  swing1LimitDegrees=45
  swing1Motion=Free
  swing2LimitDegrees=45
  swing2Motion=Free
  twistLimitDegrees=45
  twistMotion=Free
 }
]
