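# Post-process style Anim Blueprint for the female medium COD Squad Plaid body, adding secondary motion to outfit bones.
# Copies the attached parent mesh's pose, then chains a weighted knot constraint, RigidBody simulation, spring-based AnimDynamics on belt shells, constrained and simulated back shells, and an arm bullet constraint.
# EventGraph maps thigh rotation from ref pose to alphas for additive glute bone corrections, but those nodes sit outside the Output Pose chain.
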AnimBlueprint /Game/Characters/Player/Female/Medium/Bodies/F_MED_COD_Squad_Plaid/Meshes/F_MED_COD_Squad_Plaid_AnimBP
target_skeleton=&/F_MED_COD_Squad_Plaid_Skeleton
graphs [
 AnimGraph {nodes [
   n1 = "Transform (Modify) Bone - Bone: dyn_glute_l"(Alpha=Get thigh_rot_l(), Translation="-3.750,0.875,-0.9", Rotation="0.000,0.000,-55", Scale="1.25,1.25,1.25") {
    rotationMode=BMM_Additive
    rotationSpace=BCS_BoneSpace
    scaleMode=BMM_Additive
    scaleSpace=BCS_BoneSpace
    translationMode=BMM_Additive
    translationSpace=BCS_BoneSpace
    alphaBoolBlend {blendOption=Linear}
    boneToModify {boneName=dyn_glute_l}
    rotation {roll=-55}
    scale=<1.25 1.25 1.25>
    translation=<-3.75 0.875 -0.9>
   }
   "Transform (Modify) Bone - Bone: dyn_glute_r"(ComponentPose=@n1, Alpha=Get thigh_rot_r(), Translation="3.750,0,0", Rotation="0.000,0.000,-40", Scale="1.25,1.25,1.25") {
    rotationMode=BMM_Additive
    rotationSpace=BCS_BoneSpace
    scaleMode=BMM_Additive
    scaleSpace=BCS_BoneSpace
    translationMode=BMM_Additive
    translationSpace=BCS_BoneSpace
    alphaBoolBlend {blendOption=Linear}
    boneToModify {boneName=dyn_glute_r}
    rotation {roll=-40}
    scale=<1.25 1.25 1.25>
    translation {x=3.75}
   }
   n2 = "Constraint - dyn_knot (PPPPP)"!(
    ComponentPose=Local To Component(LocalPose=Copy Pose From Mesh() {bUseAttachedParent=true})
    Alpha=1
    ConstraintWeights_0=0.1165
    ConstraintWeights_1:float=0.2409
    ConstraintWeights_2:float=0.2301
    ConstraintWeights_3:float=0.4085
    ConstraintWeights_4:float=0.004
   ) -> (Pose) {
    alphaBoolBlend {blendOption=Linear}
    boneToModify {boneName=dyn_knot}
    constraintSetup [
     {targetBone {boneName=spine_01}}
     {targetBone {boneName=spine_02}}
     {targetBone {boneName=spine_03}}
     {targetBone {boneName=spine_04}}
     {targetBone {boneName=pelvis}}
    ]
    constraintWeights [1,1,1,1,1]
   }
   n3 = RigidBody(ComponentPose=@n2) {
    bClampLinearTranslationLimitToRefPose=true
    bOverrideWorldGravity=true
    overridePhysicsAsset=&/F_MED_COD_Squad_Plaid_Physics
    simulationSpace=BaseBoneSpace
    alphaBoolBlend {blendOption=Linear}
    baseBoneRef {boneName=root}
    componentAppliedLinearAccClamp=<1100 1100 1100>
    componentLinearAccScale=<0.7 0.7 0.4>
    componentLinearVelScale=<0.7 0.7 0.4>
    overrideWorldGravity {z=-980}
   }
   n4 = Anim Dynamics - Bone: dyn_belt_shell_01(ComponentPose=@n3, Alpha=0.7, ExternalForce="0,0,0") {
    angularDampingOverride=1
    angularSpringConstant=100
    bAngularSpring=true
    bGravityOverrideInSimSpace=true
    bLinearSpring=true
    bOverrideAngularDamping=true
    bOverrideLinearDamping=true
    gravityScale=0.37
    linearDampingOverride=0.9999
    linearSpringConstant=100
    simulationSpace=RootRelative
    alphaBoolBlend {blendOption=Linear}
    boundBone {boneName=dyn_belt_shell_01}
    componentAppliedLinearAccClamp=<1100 1100 1100>
    componentLinearAccScale=<0.7 0.7 0.3>
    componentLinearVelScale=<0.7 0.7 0.3>
    physicsBodyDefinitions [{
      sphereCollisionRadius=0
      boundBone {boneName=dyn_belt_shell_01}
      boxExtents=<5 2 6>
      constraintSetup {angularLimitsMax=<0 5 0>, angularLimitsMin=<0 -5 0>, angularTarget=<0 0 0>, linearAxesMax=<0.2 0 0>, linearAxesMin=<-0.2 0 0>}
      localJointOffset=<-1 0 0>
     }]
   }
   n5 = Anim Dynamics - Bone: dyn_belt_shell_02(ComponentPose=@n4, Alpha=0.9, ExternalForce="0,0,0") {
    angularDampingOverride=1
    angularSpringConstant=100
    bAngularSpring=true
    bGravityOverrideInSimSpace=true
    bLinearSpring=true
    bOverrideAngularDamping=true
    bOverrideLinearDamping=true
    gravityScale=0.35
    linearDampingOverride=0.9999
    linearSpringConstant=100
    simulationSpace=RootRelative
    alphaBoolBlend {blendOption=Linear}
    boundBone {boneName=dyn_belt_shell_02}
    componentAppliedLinearAccClamp=<1100 1100 1100>
    componentLinearAccScale=<0.7 0.7 0.3>
    componentLinearVelScale=<0.7 0.7 0.3>
    physicsBodyDefinitions [{
      sphereCollisionRadius=0
      boundBone {boneName=dyn_belt_shell_02}
      boxExtents=<2.25 2.5 6>
      constraintSetup {angularLimitsMax=<0 5 0>, angularLimitsMin=<0 -5 0>, angularTarget=<0 0 0>, linearAxesMax=<0.2 0 0>, linearAxesMin=<-0.2 0 0>}
      localJointOffset=<-1.5 0 0>
     }]
   }
   n6 = Anim Dynamics - Bone: dyn_belt_shell_03(ComponentPose=@n5, Alpha=0.95, ExternalForce="0,0,0") {
    angularDampingOverride=1
    angularSpringConstant=100
    bAngularSpring=true
    bGravityOverrideInSimSpace=true
    bLinearSpring=true
    bOverrideAngularDamping=true
    bOverrideLinearDamping=true
    gravityScale=0.35
    linearDampingOverride=0.9999
    linearSpringConstant=15
    simulationSpace=RootRelative
    alphaBoolBlend {blendOption=Linear}
    boundBone {boneName=dyn_belt_shell_03}
    componentAppliedLinearAccClamp=<1100 1100 1100>
    componentLinearAccScale=<0.7 0.7 0.3>
    componentLinearVelScale=<0.7 0.7 0.3>
    physicsBodyDefinitions [{
      sphereCollisionRadius=0
      boundBone {boneName=dyn_belt_shell_03}
      boxExtents=<5 5 3>
      constraintSetup {angularLimitsMax=<0 5 0>, angularLimitsMin=<0 -5 0>, angularTarget=<0 0 0>, linearAxesMax=<0.2 0 0>, linearAxesMin=<-0.2 0 0>}
      localJointOffset=<-2 0 0>
     }]
   }
   n7 = Anim Dynamics - Bone: dyn_belt_shell_04(ComponentPose=@n6, Alpha=0.7, ExternalForce="0,0,0") {
    angularDampingOverride=1
    angularSpringConstant=100
    bAngularSpring=true
    bGravityOverrideInSimSpace=true
    bLinearSpring=true
    bOverrideAngularDamping=true
    bOverrideLinearDamping=true
    gravityScale=0.33
    linearDampingOverride=0.9999
    linearSpringConstant=100
    simulationSpace=RootRelative
    alphaBoolBlend {blendOption=Linear}
    boundBone {boneName=dyn_belt_shell_04}
    componentAppliedLinearAccClamp=<1100 1100 1100>
    componentLinearAccScale=<0.7 0.7 0.3>
    componentLinearVelScale=<0.7 0.7 0.3>
    physicsBodyDefinitions [{
      sphereCollisionRadius=0
      boundBone {boneName=dyn_belt_shell_04}
      boxExtents=<5 3 5>
      constraintSetup {angularLimitsMax=<0 5 0>, angularLimitsMin=<0 -5 0>, angularTarget=<0 0 0>, linearAxesMax=<0.2 0 0>, linearAxesMin=<-0.2 0 0>}
      localJointOffset=<-2 0 0>
     }]
   }
   n8 = Anim Dynamics - Bone: dyn_belt_shell_05(ComponentPose=@n7, Alpha=0.6, ExternalForce="0,0,0") {
    angularDampingOverride=1
    angularSpringConstant=100
    bAngularSpring=true
    bGravityOverrideInSimSpace=true
    bLinearSpring=true
    bOverrideAngularDamping=true
    bOverrideLinearDamping=true
    gravityScale=0.35
    linearDampingOverride=0.9999
    linearSpringConstant=100
    simulationSpace=RootRelative
    alphaBoolBlend {blendOption=Linear}
    boundBone {boneName=dyn_belt_shell_05}
    componentAppliedLinearAccClamp=<1100 1100 1100>
    componentLinearAccScale=<0.7 0.7 0.3>
    componentLinearVelScale=<0.7 0.7 0.3>
    physicsBodyDefinitions [{
      sphereCollisionRadius=0
      boundBone {boneName=dyn_belt_shell_05}
      boxExtents=<3 3 5>
      constraintSetup {angularLimitsMax=<0 5 0>, angularLimitsMin=<0 -5 0>, angularTarget=<0 0 0>, linearAxesMax=<0.2 0 0>, linearAxesMin=<-0.2 0 0>}
      localJointOffset=<-1.25 0 0>
     }]
   }
   n9 = Anim Dynamics - Bone: dyn_belt_shell_06(ComponentPose=@n8, Alpha=0.8, ExternalForce="0,0,0") {
    angularDampingOverride=1
    angularSpringConstant=100
    bAngularSpring=true
    bGravityOverrideInSimSpace=true
    bLinearSpring=true
    bOverrideAngularDamping=true
    bOverrideLinearDamping=true
    gravityScale=0.28
    linearDampingOverride=0.9999
    linearSpringConstant=100
    simulationSpace=RootRelative
    alphaBoolBlend {blendOption=Linear}
    boundBone {boneName=dyn_belt_shell_06}
    componentAppliedLinearAccClamp=<1100 1100 1100>
    componentLinearAccScale=<0.7 0.7 0.3>
    componentLinearVelScale=<0.7 0.7 0.3>
    physicsBodyDefinitions [{
      sphereCollisionRadius=0
      boundBone {boneName=dyn_belt_shell_06}
      boxExtents=<2 2 2>
      constraintSetup {angularLimitsMax=<0 5 0>, angularLimitsMin=<0 -5 0>, angularTarget=<0 0 0>, linearAxesMax=<0.2 0 0>, linearAxesMin=<-0.2 0 0>}
      localJointOffset=<-1.5 0 0>
     }]
   }
   n10 = Anim Dynamics - Bone: dyn_belt_shell_07(ComponentPose=@n9, Alpha=0.55, ExternalForce="0,0,0") {
    angularDampingOverride=1
    angularSpringConstant=100
    bAngularSpring=true
    bGravityOverrideInSimSpace=true
    bLinearSpring=true
    bOverrideAngularDamping=true
    bOverrideLinearDamping=true
    gravityScale=0.32
    linearDampingOverride=0.9999
    linearSpringConstant=100
    simulationSpace=RootRelative
    alphaBoolBlend {blendOption=Linear}
    boundBone {boneName=dyn_belt_shell_07}
    componentAppliedLinearAccClamp=<1100 1100 1100>
    componentLinearAccScale=<0.7 0.7 0.3>
    componentLinearVelScale=<0.7 0.7 0.3>
    physicsBodyDefinitions [{
      sphereCollisionRadius=0
      boundBone {boneName=dyn_belt_shell_07}
      boxExtents=<3 3 4>
      constraintSetup {angularLimitsMax=<0 5 0>, angularLimitsMin=<0 -5 0>, angularTarget=<0 0 0>, linearAxesMax=<0.2 0 0>, linearAxesMin=<-0.2 0 0>}
      localJointOffset=<-1.5 0 0>
     }]
   }
   n11 = Anim Dynamics - Bone: dyn_belt_shell_08(ComponentPose=@n10, Alpha=0.9, ExternalForce="0,0,0") {
    angularDampingOverride=1
    angularSpringConstant=100
    bAngularSpring=true
    bGravityOverrideInSimSpace=true
    bLinearSpring=true
    bOverrideAngularBias=true
    bOverrideLinearDamping=true
    gravityScale=0.33
    linearDampingOverride=0.9999
    linearSpringConstant=100
    simulationSpace=RootRelative
    alphaBoolBlend {blendOption=Linear}
    boundBone {boneName=dyn_belt_shell_08}
    componentAppliedLinearAccClamp=<1100 1100 1100>
    componentLinearAccScale=<0.7 0.7 0.3>
    componentLinearVelScale=<0.7 0.7 0.3>
    physicsBodyDefinitions [{
      sphereCollisionRadius=0
      boundBone {boneName=dyn_belt_shell_08}
      boxExtents=<5 4 4>
      constraintSetup {angularLimitsMax=<0 3 0>, angularLimitsMin=<0 -3 0>, angularTarget=<0 0 0>, linearAxesMax=<0.2 0 0>, linearAxesMin=<-0.2 0 0>}
      localJointOffset=<-1 0 0>
     }]
   }
   n12 = "Constraint - dyn_back_shell_01 (PPPPP)"!(
    ComponentPose=@n11
    Alpha=1
    ConstraintWeights_0=0.0067
    ConstraintWeights_1:float=0.106
    ConstraintWeights_2:float=0.4321
    ConstraintWeights_3:float=0.0463
    ConstraintWeights_4:float=0.4088
   ) -> (Pose) {
    alphaBoolBlend {blendOption=Linear}
    boneToModify {boneName=dyn_back_shell_01}
    constraintSetup [
     {targetBone {boneName=spine_03}}
     {targetBone {boneName=spine_04}}
     {targetBone {boneName=spine_05}}
     {targetBone {boneName=upperarm_twist_01_l}}
     {targetBone {boneName=clavicle_l}}
    ]
    constraintWeights [1,1,1,1,1]
   }
   n13 = "Constraint - dyn_back_shell_02 (PPPPPP)"!(
    ComponentPose=@n12
    Alpha=1
    ConstraintWeights_0=0.2344
    ConstraintWeights_1:float=0.0164
    ConstraintWeights_2:float=0.0275
    ConstraintWeights_3:float=0.215
    ConstraintWeights_4:float=0.491
    ConstraintWeights_5:float=0.0156
   ) -> (Pose) {
    alphaBoolBlend {blendOption=Linear}
    boneToModify {boneName=dyn_back_shell_02}
    constraintSetup [
     {targetBone {boneName=clavicle_l}}
     {targetBone {boneName=spine_02}}
     {targetBone {boneName=spine_03}}
     {targetBone {boneName=spine_04}}
     {targetBone {boneName=spine_05}}
     {targetBone {boneName=upperarm_twist_01_l}}
    ]
    constraintWeights [1,1,1,1,1,1]
   }
   n14 = "Constraint - dyn_back_shell_03 (PPPPP)"!(
    ComponentPose=@n13
    Alpha=1
    ConstraintWeights_0=0.025
    ConstraintWeights_1:float=0.04
    ConstraintWeights_2:float=0.304
    ConstraintWeights_3:float=0.5404
    ConstraintWeights_4:float=0.0906
   ) -> (Pose) {
    alphaBoolBlend {blendOption=Linear}
    boneToModify {boneName=dyn_back_shell_03}
    constraintSetup [
     {targetBone {boneName=spine_02}}
     {targetBone {boneName=spine_03}}
     {targetBone {boneName=spine_04}}
     {targetBone {boneName=spine_05}}
     {targetBone {boneName=clavicle_l}}
    ]
    constraintWeights [1,1,1,1,1]
   }
   n15 = Anim Dynamics - Bone: dyn_back_shell_01(ComponentPose=@n14, Alpha=1, ExternalForce="0,0,0") {
    angularDampingOverride=1
    angularSpringConstant=100
    bAngularSpring=true
    bGravityOverrideInSimSpace=true
    bLinearSpring=true
    bOverrideAngularDamping=true
    gravityScale=0.35
    linearDampingOverride=0.9999
    linearSpringConstant=100
    simulationSpace=RootRelative
    alphaBoolBlend {blendOption=Linear}
    boundBone {boneName=dyn_back_shell_01}
    componentAppliedLinearAccClamp=<1100 1100 1100>
    componentLinearAccScale=<0.7 0.7 0.3>
    componentLinearVelScale=<0.7 0.7 0.3>
    physicsBodyDefinitions [{
      sphereCollisionRadius=0
      boundBone {boneName=dyn_back_shell_01}
      boxExtents=<5 5 5>
      constraintSetup {angularLimitsMax=<0 1 0>, angularLimitsMin=<0 -1 0>, angularTarget=<0 0 0>, linearAxesMax=<0.05 0 0>, linearAxesMin=<-0.05 0 0>}
      localJointOffset=<1 0 0>
     }]
   }
   n16 = Anim Dynamics - Bone: dyn_back_shell_02(ComponentPose=@n15, Alpha=1, ExternalForce="0,0,0") {
    angularDampingOverride=1
    angularSpringConstant=100
    bAngularSpring=true
    bGravityOverrideInSimSpace=true
    bLinearSpring=true
    bOverrideAngularDamping=true
    gravityScale=0.35
    linearDampingOverride=0.9999
    linearSpringConstant=100
    simulationSpace=RootRelative
    alphaBoolBlend {blendOption=Linear}
    boundBone {boneName=dyn_back_shell_02}
    componentAppliedLinearAccClamp=<1100 1100 1100>
    componentLinearAccScale=<0.7 0.7 0.3>
    componentLinearVelScale=<0.7 0.7 0.3>
    physicsBodyDefinitions [{
      sphereCollisionRadius=0
      boundBone {boneName=dyn_back_shell_02}
      boxExtents=<5 5 5>
      constraintSetup {angularLimitsMax=<0 1 0>, angularLimitsMin=<0 -1 0>, angularTarget=<0 0 0>, linearAxesMax=<0.05 0 0>, linearAxesMin=<-0.05 0 0>}
      localJointOffset=<1 0 0>
     }]
   }
   n17 = Anim Dynamics - Bone: dyn_back_shell_03(ComponentPose=@n16, Alpha=1, ExternalForce="0,0,0") {
    angularDampingOverride=1
    angularSpringConstant=100
    bAngularSpring=true
    bGravityOverrideInSimSpace=true
    bLinearSpring=true
    bOverrideAngularDamping=true
    gravityScale=0.35
    linearDampingOverride=0.9999
    linearSpringConstant=100
    simulationSpace=RootRelative
    alphaBoolBlend {blendOption=Linear}
    boundBone {boneName=dyn_back_shell_03}
    componentAppliedLinearAccClamp=<1100 1100 1100>
    componentLinearAccScale=<0.7 0.7 0.3>
    componentLinearVelScale=<0.7 0.7 0.3>
    physicsBodyDefinitions [{
      sphereCollisionRadius=0
      boundBone {boneName=dyn_back_shell_03}
      boxExtents=<5 5 5>
      constraintSetup {angularLimitsMax=<0 1 0>, angularLimitsMin=<0 -1 0>, angularTarget=<0 0 0>, linearAxesMax=<0.05 0 0>, linearAxesMin=<-0.05 0 0>}
      localJointOffset=<1 0 0>
     }]
   }
   n18 = "Constraint - dyn_arm_bullet_01 (PPP)"!(ComponentPose=@n17, Alpha=1, ConstraintWeights_0=0.1, ConstraintWeights_1:float=0.1, ConstraintWeights_2:float=0.8) -> (Pose) {
    alphaBoolBlend {blendOption=Linear}
    boneToModify {boneName=dyn_arm_bullet_01}
    constraintSetup [{targetBone {boneName=upperarm_twist_02_l}}, {targetBone {boneName=upperarm_l}}, {targetBone {boneName=upperarm_twist_01_l}}]
    constraintWeights [1,1,1]
   }
   Output Pose(Result=Component To Local(ComponentPose=@n18)) {layerGroup=None}
  ]}
 EventGraph {nodes [
   Event BlueprintUpdateAnimation()
   > Get!(execute:Exec) -> (then:Exec, else:Exec, Skeletal_Mesh:Skeletal Mesh Component Object Reference)
   > Sequence()
   n1 = Get Skeletal_Mesh() -> (Skeletal_Mesh:Skeletal Mesh Component Object Reference)
   n2 = BreakTransform(InTransform=GetDeltaTransformFromRefPose(self=@n1, BoneName=thigh_l, BaseName=pelvis))
   > Set thigh_rot_l(thigh_rot_l=MapRangeClamped(Value=BreakRotator(InRot=@n2.Rotation).Yaw, InRangeA=-20, InRangeB=-100))
   n3 = BreakTransform(InTransform=GetDeltaTransformFromRefPose(self=@n1, BoneName=thigh_r, BaseName=pelvis))
   > Set thigh_rot_r(thigh_rot_r=MapRangeClamped(Value=BreakRotator(InRot=@n3.Rotation).Yaw, InRangeA=-20, InRangeB=-100))
   Event BlueprintInitializeAnimation()
   > Set Skeletal_Mesh(Skeletal_Mesh:Skeletal Mesh Component Object Reference=GetOwnerSkeletalMeshComponent()) -> (Output_Get:Skeletal Mesh Component Object Reference)
   Reroute Node()
   n4 = Get Skeletal_Mesh()
   n5 = BreakTransform(InTransform=GetDeltaTransformFromRefPose(self=@n4, BoneName=calf_l, BaseName=thigh_l))
   > Set calf_rot_l(calf_rot_l:double=MapRangeClamped(Value=BreakRotator(InRot=@n5.Rotation).Yaw, InRangeA=0, InRangeB=115)) -> (Output_Get:double)
   n6 = BreakTransform(InTransform=GetDeltaTransformFromRefPose(self=@n4, BoneName=thigh_r, BaseName=calf_r))
   > Set calf_rot_r(calf_rot_r:double=MapRangeClamped(Value=BreakRotator(InRot=@n6.Rotation).Yaw, InRangeA=0, InRangeB=-115)) -> (Output_Get:double)
  ]}
]
variables [
 calf_rot_l {type=number}
 calf_rot_r {type=number}
 Skeletal_Mesh {type=SkeletalMeshComponent}
 thigh_rot_l {type=number}
 thigh_rot_r {type=number}
]
