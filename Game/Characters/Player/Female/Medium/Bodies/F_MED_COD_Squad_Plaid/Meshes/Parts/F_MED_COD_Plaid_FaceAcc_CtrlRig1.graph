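# Control Rig for the female Squad Plaid face accessory that procedurally adjusts ponytail and hair bones.
# Right upper arm and head rotations relative to spine_05 drive smoothed, clamped weights for additive offsets on dyn_parent_pony.
# That bone's radial distance around neck_01 is then clamped, keeping it outside the neck.
# Aim steps before and after the clamp rotate dyn_driver_hairy using pony bone transforms, and dyn_hairy_02 finally snaps to dyn_parent_pony's position.
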ControlRigBlueprint /Game/Characters/Player/Female/Medium/Bodies/F_MED_COD_Squad_Plaid/Meshes/Parts/F_MED_COD_Plaid_FaceAcc_CtrlRig1
graphs [
 RigVMFunctionLibrary {links [], nodes []}
 RigVMModel {nodes [
   n1 = &RigUnit_BeginExecution()
   /Script/ControlRig.RigUnit_DebugTransformMutable(
    Color=(R=1,G=0,B=0,A=1)
    DebugDrawSettings=(DepthPriority=SDPG_Foreground,Lifetime=-1)
    Mode=Axes
    Scale=10
    Space=None
    Thickness=0
    WorldOffset=(Rotation=(X=0,Y=0,Z=0,W=1),Translation=(X=0,Y=0,Z=0),Scale3D=(X=1,Y=1,Z=1))
    bEnabled=false
    ExecutePin=@n1.ExecutePin
    Transform=&RigUnit_GetBoneTransform(Bone=upperarm_r).Transform
   ) {is=struct RigVMUnitNode}
   n2 = &RigVMFunction_MathQuaternionToEuler(RotationOrder=YZX, Value=&RigUnit_GetRelativeBoneTransform(Bone=upperarm_r, Space=spine_05).Transform.Rotation)
   > /Script/ControlRig.RigUnit_ModifyBoneTransforms(
    BoneToModify=((Bone="dyn_parent_pony",Transform=(Rotation=(X=0,Y=0,Z=0,W=1),Translation=(X=4,Y=0,Z=0.5),Scale3D=(X=1,Y=1,Z=1))))
    Mode=AdditiveLocal
    WeightMaximum=1
    WeightMinimum=0
    Weight=/Script/RigVM.RigVMFunction_AccumulateFloatLerp(
     Blend=10
     InitialValue=0
     bIntegrateDeltaTime=true
     TargetValue=&RigVMFunction_MathFloatRemap(SourceMaximum=-70, SourceMinimum=-10, TargetMinimum=0, bClamp=true, Value=@n2.Result.Y).Result
    ) {is=struct RigVMUnitNode, template="AccumulateLerp::Execute(in TargetValue,in InitialValue,in Blend,in bIntegrateDeltaTime,out Result)"}.Result
    BoneToModify.0.Transform.Rotation=&RigVMFunction_MathQuaternionFromEuler(Euler=(X=-10,Y=-10,Z=-30)).Result
   ) {is=struct RigVMUnitNode}
   n3 = &RigVMFunction_MathQuaternionToEuler(RotationOrder=XZY, Value=&RigUnit_GetRelativeBoneTransform(Bone=head, Space=spine_05).Transform.Rotation)
   > &RigUnit_ModifyBoneTransforms(
    BoneToModify=((Bone="dyn_parent_pony",Transform=(Rotation=(X=0,Y=0,Z=0,W=1),Translation=(X=0.5,Y=2.5,Z=0),Scale3D=(X=1,Y=1,Z=1))))
    Mode=AdditiveLocal
    WeightMaximum=1
    WeightMinimum=0
    Weight=&RigVMFunction_AccumulateFloatLerp(
     Blend=10
     InitialValue=0
     bIntegrateDeltaTime=true
     TargetValue=&RigVMFunction_MathFloatRemap(SourceMaximum=-50, SourceMinimum=10, TargetMinimum=0, bClamp=true, Value=@n3.Result.X).Result
    ) {template="AccumulateLerp::Execute(in TargetValue,in InitialValue,in Blend,in bIntegrateDeltaTime,out Result)"}.Result
    BoneToModify.0.Transform.Rotation=&RigVMFunction_MathQuaternionFromEuler(Euler=(X=0,Y=-8,Z=20)).Result
   )
   n4 = &RigVMFunction_MathQuaternionToEuler(RotationOrder=XZY, Value=&RigUnit_GetRelativeBoneTransform(Bone=head, Space=spine_05).Transform.Rotation)
   > &RigUnit_ModifyBoneTransforms(
    BoneToModify=((Bone="dyn_parent_pony",Transform=(Rotation=(X=0,Y=0,Z=0,W=1),Translation=(X=-1.5,Y=-6,Z=1),Scale3D=(X=1,Y=1,Z=1))))
    Mode=AdditiveLocal
    WeightMaximum=1
    WeightMinimum=0
    Weight=&RigVMFunction_AccumulateFloatLerp(
     Blend=10
     InitialValue=0
     bIntegrateDeltaTime=true
     TargetValue=&RigVMFunction_MathFloatRemap(SourceMaximum=50, SourceMinimum=-10, TargetMinimum=0, bClamp=true, Value=@n4.Result.X).Result
    ) {template="AccumulateLerp::Execute(in TargetValue,in InitialValue,in Blend,in bIntegrateDeltaTime,out Result)"}.Result
    BoneToModify.0.Transform.Rotation=&RigVMFunction_MathQuaternionFromEuler(Euler=(X=30,Y=0,Z=35)).Result
   )
   n5 = &RigVMFunction_MathQuaternionToEuler(RotationOrder=XZY, Value=&RigUnit_GetRelativeBoneTransform(Bone=head, Space=spine_05).Transform.Rotation)
   > &RigUnit_ModifyBoneTransforms(
    BoneToModify=((Bone="dyn_parent_pony",Transform=(Rotation=(X=0,Y=0,Z=0,W=1),Translation=(X=5,Y=0.5,Z=-2),Scale3D=(X=1,Y=1,Z=1))))
    Mode=AdditiveLocal
    WeightMaximum=1
    WeightMinimum=0
    Weight=&RigVMFunction_AccumulateFloatLerp(
     Blend=10
     InitialValue=0
     bIntegrateDeltaTime=true
     TargetValue=&RigVMFunction_MathFloatRemap(SourceMaximum=40, SourceMinimum=-8, TargetMinimum=0, bClamp=true, Value=@n5.Result.Z).Result
    ) {template="AccumulateLerp::Execute(in TargetValue,in InitialValue,in Blend,in bIntegrateDeltaTime,out Result)"}.Result
    BoneToModify.0.Transform.Rotation=&RigVMFunction_MathQuaternionFromEuler(Euler=(X=-10,Y=-10,Z=0)).Result
   )
   n6 = &RigVMFunction_MathQuaternionToEuler(RotationOrder=XZY, Value=&RigUnit_GetRelativeBoneTransform(Bone=head, Space=spine_05).Transform.Rotation)
   > &RigUnit_ModifyBoneTransforms(
    BoneToModify=((Bone="dyn_parent_pony",Transform=(Rotation=(X=0,Y=0,Z=0,W=1),Translation=(X=0,Y=0,Z=-1),Scale3D=(X=1,Y=1,Z=1))))
    Mode=AdditiveLocal
    WeightMaximum=1
    WeightMinimum=0
    Weight=&RigVMFunction_AccumulateFloatLerp(
     Blend=10
     InitialValue=0
     bIntegrateDeltaTime=true
     TargetValue=&RigVMFunction_MathFloatRemap(SourceMaximum=-40, SourceMinimum=0, TargetMinimum=0, bClamp=true, Value=@n6.Result.Z).Result
    ) {template="AccumulateLerp::Execute(in TargetValue,in InitialValue,in Blend,in bIntegrateDeltaTime,out Result)"}.Result
    BoneToModify.0.Transform.Rotation=&RigVMFunction_MathQuaternionFromEuler(Euler=(X=13,Y=0,Z=0)).Result
   )
   > &RigUnit_DebugTransformMutable(
    Color=(R=1,G=0,B=0,A=1)
    DebugDrawSettings=(DepthPriority=SDPG_Foreground,Lifetime=-1)
    Mode=Axes
    Scale=50
    Space=dyn_driver_hairy
    Thickness=0
    Transform=(Rotation=(X=0,Y=0,Z=0,W=1),Translation=(X=0,Y=0,Z=0),Scale3D=(X=1,Y=1,Z=1))
    WorldOffset=(Rotation=(X=0,Y=0,Z=0,W=1),Translation=(X=0,Y=0,Z=0),Scale3D=(X=1,Y=1,Z=1))
    bEnabled=false
   )
   n7 = &RigUnit_GetBoneTransform(Bone=dyn_driver_hairy)
   n8 = /Script/RigVM.RigVMFunction_MathQuaternionGetAxis(Axis=X, Quaternion=@n7.Transform.Rotation) {is=struct RigVMUnitNode, template="Axis::Execute(in Quaternion,in Axis,out Result)"}
   n9 = /Script/RigVM.RigVMFunction_MathTransformTransformVector(
    Transform=&RigUnit_GetBoneTransform(Bone=dyn_driver_pony, Space=LocalSpace).Transform
    Location=/Script/RigVM.RigVMFunction_MathVectorSelectBool(
     Condition=False
     IfFalse=(X=0,Y=0,Z=0)
     IfTrue=/Script/RigVM.RigVMFunction_MathVectorAdd(A=(X=8,Y=8,Z=-8), B=(X=0,Y=0,Z=0)) {is=struct RigVMUnitNode, template="Add::Execute(in A,in B,out Result)"}.Result
    ) {is=struct RigVMUnitNode}.Result
   ) {is=struct RigVMUnitNode}
   n10 = /Script/RigVM.RigVMFunction_MathVectorSub(B=@n7.Transform.Translation, A=@n9.Result) {is=struct RigVMUnitNode, template="Subtract::Execute(in A,in B,out Result)"}
   > /Script/ControlRig.RigUnit_SetBoneRotation(
    Bone=dyn_driver_hairy
    Space=GlobalSpace
    Weight=1
    bPropagateToChildren=True
    Rotation=/Script/RigVM.RigVMFunction_MathQuaternionMul(
     B=@n7.Transform.Rotation
     A=/Script/RigVM.RigVMFunction_MathQuaternionFromTwoVectors(A=@n8.Result, B=@n10.Result) {is=struct RigVMUnitNode, template="FromTwoVectors::Execute(in A,in B,out Result)"}.Result
    ) {is=struct RigVMUnitNode, template="Multiply::Execute(in A,in B,out Result)"}.Result
   ) {is=struct RigVMUnitNode}
   > &RigUnit_DebugTransformMutable(
    Color=(R=1,G=0,B=0,A=1)
    DebugDrawSettings=(DepthPriority=SDPG_Foreground,Lifetime=-1)
    Mode=Axes
    Scale=75
    Space=dyn_driver_hairy
    Thickness=0
    Transform=(Rotation=(X=0,Y=0,Z=0,W=1),Translation=(X=0,Y=0,Z=0),Scale3D=(X=1,Y=1,Z=1))
    WorldOffset=(Rotation=(X=0,Y=0,Z=0,W=1),Translation=(X=0,Y=0,Z=0),Scale3D=(X=1,Y=1,Z=1))
    bEnabled=false
   )
   > &RigUnit_DebugTransformMutable(
    Color=(R=1,G=0,B=0,A=1)
    DebugDrawSettings=(DepthPriority=SDPG_Foreground,Lifetime=-1)
    Mode=Axes
    Scale=55
    Space=None
    Thickness=0
    Transform=(Rotation=(X=0,Y=0,Z=0,W=1),Translation=(X=0,Y=0,Z=0),Scale3D=(X=1,Y=1,Z=1))
    WorldOffset=(Rotation=(X=0,Y=0,Z=0,W=1),Translation=(X=0,Y=0,Z=0),Scale3D=(X=1,Y=1,Z=1))
    bEnabled=false
    Transform.Translation=@n9.Result
   )
   > &RigUnit_DebugTransformMutable(
    Color=(R=1,G=0,B=0,A=1)
    DebugDrawSettings=(DepthPriority=SDPG_Foreground,Lifetime=-1)
    Mode=Axes
    Scale=10
    Space=neck_01
    Thickness=0
    Transform=(Rotation=(X=0,Y=0,Z=0,W=1),Translation=(X=0,Y=0,Z=0),Scale3D=(X=1,Y=1,Z=1))
    WorldOffset=(Rotation=(X=0,Y=0,Z=0,W=1),Translation=(X=0,Y=0,Z=0),Scale3D=(X=1,Y=1,Z=1))
    bEnabled=True
   )
   n11 = &RigUnit_GetBoneTransform(Bone=neck_01)
   n12 = &RigVMFunction_MathTransformMakeRelative(Parent=@n11.Transform, Global=&RigUnit_GetBoneTransform(Bone=dyn_parent_pony).Transform)
   n13 = /Script/RigVM.RigVMFunction_MathVectorMul(B=(X=0,Y=1,Z=1), A=@n12.Local.Translation) {is=struct RigVMUnitNode, template="Multiply::Execute(in A,in B,out Result)"}
   n14 = /Script/RigVM.RigVMFunction_MathVectorLength(Value=@n13.Result) {is=struct RigVMUnitNode, template="Length::Execute(in Value,out Result)"}
   n15 = &RigVMFunction_MathVectorAdd(
    A=&RigVMFunction_MathVectorMul(B=(X=1,Y=0,Z=0), A=@n12.Local.Translation) {template="Multiply::Execute(in A,in B,out Result)"}.Result
    B=/Script/RigVM.RigVMFunction_MathVectorScale(
     Factor=/Script/RigVM.RigVMFunction_MathFloatClamp(Maximum=10, Minimum=9, Value=@n14.Result) {is=struct RigVMUnitNode, template="Clamp::Execute(in Value,in Minimum,in Maximum,out Result)"}.Result
     Value=/Script/RigVM.RigVMFunction_MathVectorUnit(Value=@n13.Result) {is=struct RigVMUnitNode, template="Unit::Execute(in Value,out Result)"}.Result
    ) {is=struct RigVMUnitNode, template="Scale::Execute(in Value,in Factor,out Result)"}.Result
   ) {template="Add::Execute(in A,in B,out Result)"}
   > /Script/ControlRig.RigUnit_SetBoneTransform(
    Bone=dyn_parent_pony
    Space=GlobalSpace
    Weight=1
    bPropagateToChildren=True
    Transform=&RigVMFunction_MathTransformMul(
     A=(Rotation=(X=0,Y=0,Z=0,W=1),Translation=(X=0,Y=0,Z=0),Scale3D=(X=1,Y=1,Z=1))
     B=@n11.Transform
     A.Rotation=@n12.Local.Rotation
     A.Translation=@n15.Result
    ) {template="Multiply::Execute(in A,in B,out Result)"}.Result
   ) {is=struct RigVMUnitNode}
   > &RigUnit_DebugTransformMutable(
    Color=(R=1,G=0,B=0,A=1)
    DebugDrawSettings=(DepthPriority=SDPG_Foreground,Lifetime=-1)
    Mode=Axes
    Scale=8
    Space=neck_01
    Thickness=0.4
    Transform=(Rotation=(X=0,Y=0,Z=0,W=1),Translation=(X=0,Y=0,Z=3),Scale3D=(X=1,Y=1,Z=1))
    WorldOffset=(Rotation=(X=0,Y=0,Z=0,W=1),Translation=(X=0,Y=0,Z=0),Scale3D=(X=1,Y=1,Z=1))
    bEnabled=True
    Transform.Translation=@n15.Result
   )
   > /Script/ControlRig.RigUnit_DebugArc(
    Color=(R=1,G=0,B=0,A=1)
    DebugDrawSettings=(DepthPriority=SDPG_Foreground,Lifetime=-1)
    Detail=16
    MaximumDegrees=360
    MinimumDegrees=0
    Radius=7
    Space=neck_01
    Thickness=0
    Transform=(Rotation=(X=0,Y=0,Z=0,W=1),Translation=(X=0,Y=0,Z=0),Scale3D=(X=1,Y=1,Z=1))
    WorldOffset=(Rotation=(X=0,Y=0,Z=0,W=1),Translation=(X=0,Y=0,Z=0),Scale3D=(X=1,Y=1,Z=1))
    bEnabled=True
    Transform.Rotation=&RigVMFunction_MathQuaternionFromEuler(Euler=(X=0,Y=90,Z=0)).Result
   ) {is=struct RigVMUnitNode}
   > &RigUnit_DebugTransformMutable(
    Color=(R=1,G=0,B=0,A=1)
    DebugDrawSettings=(DepthPriority=SDPG_Foreground,Lifetime=-1)
    Mode=Axes
    Scale=50
    Space=dyn_driver_hairy
    Thickness=0
    Transform=(Rotation=(X=0,Y=0,Z=0,W=1),Translation=(X=0,Y=0,Z=0),Scale3D=(X=1,Y=1,Z=1))
    WorldOffset=(Rotation=(X=0,Y=0,Z=0,W=1),Translation=(X=0,Y=0,Z=0),Scale3D=(X=1,Y=1,Z=1))
    bEnabled=false
   )
   n16 = &RigUnit_GetBoneTransform(Bone=dyn_driver_hairy)
   n17 = &RigVMFunction_MathQuaternionGetAxis(Axis=X, Quaternion=@n16.Transform.Rotation) {template="Axis::Execute(in Quaternion,in Axis,out Result)"}
   n18 = &RigVMFunction_MathTransformTransformVector(
    Transform=&RigUnit_GetBoneTransform(Bone=dyn_parent_pony, Space=LocalSpace).Transform
    Location=&RigVMFunction_MathVectorSelectBool(
     Condition=False
     IfFalse=(X=0,Y=0,Z=0)
     IfTrue=&RigVMFunction_MathVectorAdd(A=(X=8,Y=8,Z=-8), B=(X=0,Y=0,Z=0)) {template="Add::Execute(in A,in B,out Result)"}.Result
    ).Result
   )
   n19 = &RigVMFunction_MathVectorSub(B=@n16.Transform.Translation, A=@n18.Result) {template="Subtract::Execute(in A,in B,out Result)"}
   > &RigUnit_SetBoneRotation(
    Bone=dyn_driver_hairy
    Space=GlobalSpace
    Weight=1
    bPropagateToChildren=False
    Rotation=&RigVMFunction_MathQuaternionMul(
     B=@n16.Transform.Rotation
     A=&RigVMFunction_MathQuaternionFromTwoVectors(A=@n17.Result, B=@n19.Result) {template="FromTwoVectors::Execute(in A,in B,out Result)"}.Result
    ) {template="Multiply::Execute(in A,in B,out Result)"}.Result
   )
   > &RigUnit_DebugTransformMutable(
    Color=(R=1,G=0,B=0,A=1)
    DebugDrawSettings=(DepthPriority=SDPG_Foreground,Lifetime=-1)
    Mode=Axes
    Scale=75
    Space=dyn_driver_hairy
    Thickness=0
    Transform=(Rotation=(X=0,Y=0,Z=0,W=1),Translation=(X=0,Y=0,Z=0),Scale3D=(X=1,Y=1,Z=1))
    WorldOffset=(Rotation=(X=0,Y=0,Z=0,W=1),Translation=(X=0,Y=0,Z=0),Scale3D=(X=1,Y=1,Z=1))
    bEnabled=false
   )
   > &RigUnit_DebugTransformMutable(
    Color=(R=1,G=0,B=0,A=1)
    DebugDrawSettings=(DepthPriority=SDPG_Foreground,Lifetime=-1)
    Mode=Axes
    Scale=55
    Space=None
    Thickness=0
    Transform=(Rotation=(X=0,Y=0,Z=0,W=1),Translation=(X=0,Y=0,Z=0),Scale3D=(X=1,Y=1,Z=1))
    WorldOffset=(Rotation=(X=0,Y=0,Z=0,W=1),Translation=(X=0,Y=0,Z=0),Scale3D=(X=1,Y=1,Z=1))
    bEnabled=false
    Transform.Translation=@n18.Result
   )
   > &RigUnit_SetBoneTransform(
    Bone=dyn_hairy_02
    Space=GlobalSpace
    Transform=(Rotation=(X=0,Y=0,Z=0,W=1),Translation=(X=0,Y=0,Z=0),Scale3D=(X=1,Y=1,Z=1))
    Weight=1
    bPropagateToChildren=False
    Transform.Translation=&RigUnit_GetBoneTransform(Bone=dyn_parent_pony).Transform.Translation
    Transform.Rotation=&RigUnit_GetBoneTransform(Bone=dyn_hairy_02).Transform.Rotation
   )
   &RigUnit_BeginExecution()
   &RigVMFunction_MathTransformMul(
    A=(Rotation=(X=0,Y=0,Z=0,W=1),Translation=(X=0,Y=0,Z=0),Scale3D=(X=1,Y=1,Z=1))
    B=@n11.Transform
    A.Rotation=@n12.Local.Rotation
    A.Translation=/Script/RigVM.RigVMFunction_MathVectorClamp(
     Maximum=(X=100,Y=100,Z=10)
     Minimum=(X=-100,Y=-100,Z=3)
     Minimum.Z=&RigVMFunction_MathFloatAdd(A=3, B=0) {template="Add::Execute(in A,in B,out Result)"}.Result
     Value=@n12.Local.Translation
    ) {is=struct RigVMUnitNode, template="Clamp::Execute(in Value,in Minimum,in Maximum,out Result)"}.Result
   ) {template="Multiply::Execute(in A,in B,out Result)"}
   &RigUnit_ModifyBoneTransforms(
    BoneToModify=((Bone="dyn_parent_pony",Transform=(Rotation=(X=0,Y=0,Z=0,W=1),Translation=(X=0.5,Y=2.5,Z=0),Scale3D=(X=1,Y=1,Z=1))))
    ExecutePin=()
    Mode=AdditiveLocal
    Weight=1
    WeightMaximum=1
    WeightMinimum=0
    BoneToModify.0.Transform.Rotation=&RigVMFunction_MathQuaternionFromEuler(Euler=(X=0,Y=-8,Z=20)).Result
   )
   &RigUnit_SetBoneTransform(
    Bone=dyn_hairy_02
    ExecutePin=()
    Space=GlobalSpace
    Transform=(Rotation=(X=0,Y=0,Z=0,W=1),Translation=(X=0,Y=0,Z=0),Scale3D=(X=1,Y=1,Z=1))
    Weight=1
    bPropagateToChildren=False
    Transform.Translation=&RigUnit_GetBoneTransform(Bone=dyn_parent_pony).Transform.Translation
    Transform.Rotation=&RigUnit_GetBoneTransform(Bone=dyn_hairy_02).Transform.Rotation
   )
  ]}
]
hierarchy [
 root {loc=<0 0 0>, rot=<0 0 0 1>}
 pelvis {parent=root, loc=<0 -0.5301 98.0571>, rot=<0.0127 0.707 -0.0127 -0.707>}
 spine_01 {parent=pelvis, loc=<6.1208 -0.9328 0>, rot=<0 0 0.034 -0.9994>}
 spine_02 {parent=spine_01, loc=<6.5996 -0.2095 0>, rot=<0 0 0.0238 -0.9997>}
 spine_03 {parent=spine_02, loc=<6.4914 -0.1237 0>, rot=<0 0 0.0149 -0.9999>}
 spine_04 {parent=spine_03, loc=<6.9598 0.0794 0>, rot=<0 0 -0.0132 -0.9999>}
 spine_05 {parent=spine_04, loc=<6.9987 0 0>, rot=<0 0 -0.0015 -1>}
 clavicle_l {parent=spine_05, loc=<5.4831 -0.1053 -3.2946>, rot=<0.0465 -0.7289 -0.009 -0.6829>}
 clavicle_r {parent=spine_05, loc=<5.4831 -0.1053 3.2946>, rot=<0.7289 0.0465 -0.6829 0.009>}
 upperarm_r {parent=clavicle_r, loc=<-9.8658 1.2136 0.4104>, rot=<0.0164 -0.3245 -0.0269 -0.9454>}
 neck_01 {parent=spine_05, loc=<11.3138 0.1529 0>, rot=<0 0 -0.2003 -0.9797>}
 neck_02 {parent=neck_01, loc=<4.5627 0 0>, rot=<0 0 0.0504 -0.9987>}
 head {parent=neck_02, loc=<5.3914 0 0>, rot=<0 0 0.115 -0.9934>}
 dyn_parent_hairy {parent=head, loc=<5.0612 -1.3031 6.0999>, rot=<0.0029 -0.7071 -0.0029 -0.7071>}
 dyn_driver_hairy {parent=dyn_parent_hairy, loc=<0 0 0>, rot=<-0.4549 -0.6281 0.4071 -0.4826>}
 dyn_hairy_01 {parent=dyn_driver_hairy, loc=<0 0 0>, rot=<-0.1307 0.0334 -0.0126 -0.9908>}
 dyn_hairy_02 {parent=dyn_hairy_01, loc=<12.2594 0 0>, rot=<0 0 0 1>}
 dyn_parent_pony {parent=spine_04, loc=<20.5181 3.1033 7.1845>, rot=<0.0294 -0.7065 -0.0294 -0.7065>}
 dyn_driver_pony {parent=dyn_parent_pony, loc=<0 0 0>, rot=<-0.4618 -0.5354 0.6919 -0.1457>}
 dyn_pony_01 {parent=dyn_driver_pony, loc=<0 0 0>, rot=<0 0 0 1>}
 dyn_pony_02 {parent=dyn_pony_01, loc=<4.2567 0 0>, rot=<-0.2313 -0.2626 -0.1181 -0.9293>}
 dyn_pony_03 {parent=dyn_pony_02, loc=<6.4017 0 0>, rot=<0.2562 0.3136 0.1968 -0.8929>}
 dyn_pony_04 {parent=dyn_pony_03, loc=<3.252 0 0>, rot=<0 0 0 1>}
]
