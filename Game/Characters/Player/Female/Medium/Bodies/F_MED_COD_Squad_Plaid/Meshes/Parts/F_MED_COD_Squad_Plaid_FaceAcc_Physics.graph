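# Physics asset for the face accessory mesh of the female medium COD Squad Plaid player character, setting up a physics-driven ponytail.
# A two-joint chain links a small sphere on dyn_parent_pony to capsules on dyn_driver_pony and then dyn_pony_03, with limited swing and very tight twist.
# An unconstrained spine_05 body combines three larger capsules, apparently as an upper-torso collision volume for the hair.
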PhysicsAsset /Game/Characters/Player/Female/Medium/Bodies/F_MED_COD_Squad_Plaid/Meshes/Parts/F_MED_COD_Squad_Plaid_FaceAcc_Physics
bodies {
 dyn_driver_pony [{shapeType=Capsule, length=3.5, radius=2.505, center=<3.0472 0 -0.0001>, rotation=<rot 0 -90 90.0005>}]
 dyn_parent_pony [{shapeType=Sphere, radius=0.505}]
 dyn_pony_03 [{
   shapeType=Capsule
   length=2.8062
   radius=2.9148
   shapeName=dyn_pony_03_capsule
   center=<1.2492 0.4172 -0.8895>
   rotation=<rot -25.8849 -90.0001 85.4523>
  }]
 spine_05 [
  {shapeType=Capsule, length=0, radius=24.755, center=<-10.7133 -13.2153 0>, rotation=<rot 0 -90 0>}
  {shapeType=Capsule, length=15, radius=15, shapeName=spine_05_capsule, center=<0 -8.5175 0>}
  {shapeType=Capsule, length=15, radius=7, shapeName=spine_05_capsule, center=<0 6.0161 -2.5611>, rotation=<rot 90 0 0>}
 ]
}
constraints [
 {bone1Name=dyn_driver_pony, bone2Name=dyn_parent_pony, swing1LimitDegrees=16.4286, swing2LimitDegrees=32.5714, twistLimitDegrees=3}
 {bone1Name=dyn_pony_03, bone2Name=dyn_driver_pony, swing1LimitDegrees=15, swing2LimitDegrees=15, twistLimitDegrees=1}
]
