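# Anim Blueprint for the Squad Plaid female face accessory mesh, adding secondary motion to its ponytail bones.
# Copies the attached parent's pose, then applies parent-bone constraints and offsets, a Control Rig, RigidBody and chained AnimDynamics.
# Parachute state drives the spine constraint and offset alpha, while targeting damps RigidBody and dyn_pony_02 simulation.
# The cached result gets extra bone offsets in the front end; the dyn_pony_01 dynamics node is unconnected.
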AnimBlueprint /Game/Characters/Player/Female/Medium/Bodies/F_MED_COD_Squad_Plaid/Meshes/Parts/F_MED_COD_Squad_Plaid_FaceAcc_AnimBP
target_skeleton=&/F_MED_COD_Squad_Plaid_FaceAcc_Skeleton
graphs [
 AnimGraph {nodes [
   Anim Dynamics - Bone: dyn_pony_01(Alpha=0.45, ExternalForce="0,0,0") {
    angularDampingOverride=1
    angularSpringConstant=1500
    bAngularSpring=true
    bGravityOverrideInSimSpace=true
    bOverrideAngularDamping=true
    gravityScale=0.35
    simulationSpace=RootRelative
    alphaBoolBlend {blendOption=Linear}
    alphaScaleBiasClamp {bInterpResult=true}
    boundBone {boneName=dyn_pony_01}
    componentAppliedLinearAccClamp=<1100 1100 1100>
    componentLinearAccScale=<0.7 0.7 0.4>
    componentLinearVelScale=<0.7 0.7 0.4>
    physicsBodyDefinitions [{
      sphereCollisionRadius=0
      boundBone {boneName=dyn_pony_01}
      boxExtents=<10 20 20>
      constraintSetup {angularLimitsMax=<0 60 0>, angularLimitsMin=<0 -10 -10>, angularTarget=<0 0 0>}
      localJointOffset=<7 0 -1>
     }]
   }
   n1 = "Transform (Modify) Bone - Bone: dyn_driver_pony"!(ComponentPose=Local To Component(LocalPose=Use cached pose 'Pose'()), Alpha=1, Rotation="-15,0,0") -> (Pose) {
    rotationMode=BMM_Additive
    rotationSpace=BCS_BoneSpace
    alphaBoolBlend {blendOption=Linear}
    boneToModify {boneName=dyn_driver_pony}
    rotation {pitch=-15}
   }
   n2 = "Transform (Modify) Bone - Bone: dyn_parent_pony"!(ComponentPose=@n1, Alpha=1, Translation="3,-0.000,-0.000", Rotation="0,0,0") -> (Pose) {
    rotationMode=BMM_Additive
    rotationSpace=BCS_BoneSpace
    translationMode=BMM_Additive
    translationSpace=BCS_BoneSpace
    alphaBoolBlend {blendOption=Linear}
    boneToModify {boneName=dyn_parent_pony}
    translation {x=3}
   }
   Output Pose(Result=Blend Poses by bool(BlendPose_0=Component To Local(ComponentPose=@n2), BlendPose_1=Use cached pose 'Pose'(), bActiveValue=Get bIsInFrontEnd()) {blendType=Linear}) {layerGroup=None}
   n3 = Get Mesh_To_Copy() -> (Mesh_To_Copy:Skeletal Mesh Component Object Reference)
   n4 = Get skydiving() -> (skydiving:double)
   n5 = "Constraint - dyn_parent_pony (P)"(
    ComponentPose=Local To Component(LocalPose=Copy Pose From Mesh(SourceMeshComponent=@n3) {bUseAttachedParent=true})
    Alpha=@n4
    ConstraintWeights_0=1
   ) {
    alphaBoolBlend {blendOption=Linear}
    boneToModify {boneName=dyn_parent_pony}
    constraintSetup [{targetBone {boneName=spine_05}}]
    constraintWeights [1]
   }
   n6 = "Transform (Modify) Bone - Bone: dyn_parent_pony"!(ComponentPose=@n5, Alpha=@n4, Translation="-3.5,0,0", Rotation="0,0,-20", Scale:Vector="1,1,1") -> (Pose) {
    rotationMode=BMM_Additive
    rotationSpace=BCS_BoneSpace
    translationMode=BMM_Additive
    translationSpace=BCS_BoneSpace
    alphaBoolBlend {blendOption=Linear}
    boneToModify {boneName=dyn_parent_pony}
    rotation {roll=-20}
    translation {x=-3.5}
   }
   n7 = "Constraint - dyn_parent_pony (P)"!(ComponentPose=@n6, bAlphaBoolEnabled:bool=True, ConstraintWeights_0=1) -> (Pose) {
    alphaInputType=Bool
    alphaBoolBlend {blendOption=Linear}
    boneToModify {boneName=dyn_parent_pony}
    constraintSetup [{targetBone {boneName=neck_01}}]
    constraintWeights [1]
   }
   n8 = Control Rig(Source=Component To Local(ComponentPose=@n7)) {
    alphaBoolBlend {blendOption=Linear}
    controlRigAssetReference {blueprintRigClass=&/F_MED_COD_Plaid_FaceAcc_CtrlRig1.F_MED_COD_Plaid_FaceAcc_CtrlRig1_C}
   }
   n9 = Get targeting_alpha() -> (targeting_alpha:double)
   n10 = RigidBody(ComponentPose=Local To Component(LocalPose=@n8), Alpha=@n9) {
    bClampLinearTranslationLimitToRefPose=true
    bOverrideWorldGravity=true
    overridePhysicsAsset=&/F_MED_COD_Squad_Plaid_FaceAcc_Physics
    simulationSpace=BaseBoneSpace
    alphaBoolBlend {blendOption=Linear}
    alphaScaleBiasClamp {bInterpResult=true, interpSpeedDecreasing=15, interpSpeedIncreasing=15}
    baseBoneRef {boneName=spine_01}
    componentAppliedLinearAccClamp=<1100 1100 1100>
    componentLinearAccScale=<0.7 0.7 0.4>
    componentLinearVelScale=<0.7 0.7 0.4>
    overrideWorldGravity {z=-980}
   }
   n11 = Anim Dynamics - Bone: dyn_pony_02(ComponentPose=@n10, Alpha=@n9, ExternalForce="0,0,0") {
    angularDampingOverride=1
    angularSpringConstant=1500
    bAngularSpring=true
    bGravityOverrideInSimSpace=true
    bOverrideAngularDamping=true
    gravityScale=0.35
    simulationSpace=RootRelative
    alphaBoolBlend {blendOption=Linear}
    alphaScaleBiasClamp {bInterpResult=true}
    boundBone {boneName=dyn_pony_02}
    componentAppliedLinearAccClamp=<1100 1100 1100>
    componentLinearAccScale=<0.7 0.7 0.4>
    componentLinearVelScale=<0.7 0.7 0.4>
    physicsBodyDefinitions [{
      sphereCollisionRadius=0
      boundBone {boneName=dyn_pony_02}
      boxExtents=<10 20 20>
      constraintSetup {angularLimitsMax=<0 65 0>, angularLimitsMin=<0 -10 -10>, angularTarget=<0 0 0>}
      localJointOffset=<5 0 2>
     }]
   }
   n12 = Anim Dynamics - Bone: dyn_pony_03(ComponentPose=@n11, Alpha=0.65, ExternalForce="0,0,0") {
    angularDampingOverride=1
    angularSpringConstant=1500
    bAngularSpring=true
    bGravityOverrideInSimSpace=true
    bOverrideAngularDamping=true
    gravityScale=0.35
    simulationSpace=RootRelative
    alphaBoolBlend {blendOption=Linear}
    boundBone {boneName=dyn_pony_03}
    componentAppliedLinearAccClamp=<1100 1100 1100>
    componentLinearAccScale=<0.7 0.7 0.4>
    componentLinearVelScale=<0.7 0.7 0.4>
    physicsBodyDefinitions [{
      sphereCollisionRadius=0
      boundBone {boneName=dyn_pony_03}
      boxExtents=<8 15 15>
      constraintSetup {angularLimitsMax=<0 55 0>, angularLimitsMin=<0 -5 -40>, angularTarget=<0 0 0>}
      localJointOffset=<6 0 0>
     }]
   }
   n13 = Anim Dynamics - Bone: dyn_pony_04(ComponentPose=@n12, Alpha=0.85, ExternalForce="0,0,0") {
    angularDampingOverride=1
    angularSpringConstant=1500
    bAngularSpring=true
    bGravityOverrideInSimSpace=true
    bOverrideAngularDamping=true
    gravityScale=0.31
    simulationSpace=RootRelative
    alphaBoolBlend {blendOption=Linear}
    boundBone {boneName=dyn_pony_04}
    componentAppliedLinearAccClamp=<1100 1100 1100>
    componentLinearAccScale=<0.7 0.7 0.4>
    componentLinearVelScale=<0.7 0.7 0.4>
    physicsBodyDefinitions [{
      sphereCollisionRadius=0
      boundBone {boneName=dyn_pony_04}
      boxExtents=<5 15 15>
      constraintSetup {angularLimitsMax=<0 40 0>, angularLimitsMin=<0 -10 -40>, angularTarget=<0 0 0>}
      localJointOffset=<3 0 0>
     }]
   }
   Save cached pose 'Pose'(Pose=Component To Local(ComponentPose=@n13)) {cacheName=Pose}
  ]}
 EventGraph {nodes [
   Event BlueprintUpdateAnimation()
   > Get!(execute:Exec) -> (then:Exec, else:Exec, Mesh_To_Copy:Skeletal Mesh Component Object Reference)
   > n1 = Sequence()
   > n2 = Branch(Condition=Get bIsParachuteOpened())
   > Set skydiving(skydiving:double=1) -> (Output_Get:double)
   Set skydiving(execute=@n2.else, skydiving=0)
   n3 = Branch(execute=@n1.then_1, Condition=Get bIsTargeting())
   > Set targeting_alpha(targeting_alpha:double=0.15) -> (Output_Get:double)
   Set targeting_alpha(execute=@n3.else, targeting_alpha=0.7)
   Event BlueprintInitializeAnimation()
   > Set Mesh_To_Copy(Mesh_To_Copy:Skeletal Mesh Component Object Reference=GetOwnerSkeletalMeshComponent()) -> (Output_Get:Skeletal Mesh Component Object Reference)
   TryGetPawnOwner()
  ]}
]
variables [
 Mesh_To_Copy {type=SkeletalMeshComponent}
 skydiving {type=number}
 targeting_alpha {type=number}
]
